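# Translucent UI material for the EasyLootGrid HUD prototype, drawing a configurable fill with optional borders and icon.
# Static switches choose a flat color, linear gradient or radial gradient, with an optional secondary color.
# Four rotated, offset UV ramps are sharpened and summed into the border mask, while a scalable, movable texture supplies the icon.
# Named reroutes carry these layers into combined color and opacity results, with an overall opacity multiplier.
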Material /Game/Athena/HUD/Bacchus/Proto_1/EasyLootGrid/Art/M_EasyLootGrid
graph {
 properties {blendMode=BLEND_Translucent, materialDomain=MD_UI, pixelDepthOffsetMode=PDOM_Legacy, usedShadingModels=Default Lit}
 nodes [
  n1 = TextureCoordinate()
  n2 = Multiply(A=ScalarParameter() {defaultValue=0.03, group=Border, parameterName=Width - Left, sliderMax=1, sortPriority=35}) {constB=-1}
  n3 = ScalarParameter() {group=Border, parameterName=Angle - Left, sliderMax=360, sortPriority=35}
  n4 = Multiply(
   A=Rotator(Coordinate=Add(A=@n1, B=@n2), Time=Divide(A=@n3) {constB=360, desc=Conversion to decimal}) {speed=6.28}
   B=ScalarParameter() {defaultValue=200, group=Border, parameterName=Sharpness, sortPriority=30}
  )
  NamedRerouteDeclaration(None=LinearInterpolate(Alpha=Saturate(None=@n4) {desc=Clamp values between 0 and 1}) {constA=1, constB=0}) {name=Border Left, nodeColor=<rgba 0 0 0 1>}
  n5 = MaterialFunctionCall(UVs=TextureCoordinate(), Texture Scale=ScalarParameter() {defaultValue=1, group=Icon, parameterName=Icon Size}) {materialFunction=&ScaleUVsByCenter}
  NamedRerouteDeclaration(None=@n5.0-1 mask) {name=Icon Mask, nodeColor=<rgba 0 0 0 1>}
  n6 = Multiply(A=ScalarParameter() {group=Icon, parameterName=Icon Position X, sliderMax=1, sliderMin=-1}) {constB=-1}
  n7 = ScalarParameter() {group=Icon, parameterName=Icon Position Y, sliderMax=1, sliderMin=-1}
  n8 = TextureSampleParameter2D(UVs=Add(A=@n5.UVs, B=AppendVector(A=@n6, B=@n7))) {group=Icon, parameterName=Icon, sortPriority=2, texture=/Game/Athena/HUD/Art/BacchusHudIcons/Inventory_Pickup}
  NamedRerouteDeclaration(None=@n8.A) {name=Texture Icon Alpha, nodeColor=<rgba 0 0 0 1>}
  n9 = NamedRerouteUsage()
  n10 = NamedRerouteUsage()
  n11 = VectorParameter() {group=Border, parameterName=Border Color, sortPriority=30, defaultValue=<rgba 0.1132 0.5312 0.1028 1>}
  n12 = StaticSwitchParameter(
   True=NamedRerouteUsage()
   False=StaticSwitchParameter(True=NamedRerouteUsage(), False=VectorParameter() {parameterName=Color A, sortPriority=3, defaultValue=<rgba 0.1159 0.1884 0.9271 1>}.RGB) {parameterName=Gradient, sortPriority=5}
  ) {desc="On = Radial\r\nOff = Linear", parameterName=Radial, sortPriority=6}
  n13 = StaticSwitchParameter(True=LinearInterpolate(A=@n12, B=@n11.RGB, Alpha=NamedRerouteUsage()) {constB=0}, False=@n12) {parameterName=Borders, sortPriority=4}
  NamedRerouteDeclaration(None=StaticSwitchParameter(True=LinearInterpolate(A=@n13, B=@n9, Alpha=@n10) {constAlpha=1}, False=@n13) {parameterName=Icon, sortPriority=3}) {name=Color, nodeColor=<rgba 0 0 0 1>}
  n14 = StaticSwitchParameter(
   True=NamedRerouteUsage()
   False=StaticSwitchParameter(True=NamedRerouteUsage(), False=Constant() {r=1}) {parameterName=Gradient, sortPriority=5}
  ) {desc="On = Radial\r\nOff = Linear", parameterName=Radial, sortPriority=6}
  n15 = StaticSwitchParameter(True=Add(A=@n14, B=NamedRerouteUsage()), False=@n14) {parameterName=Borders, sortPriority=4}
  n16 = Add(A=LinearInterpolate(B=NamedRerouteUsage(), Alpha=NamedRerouteUsage()), B=@n15)
  NamedRerouteDeclaration(None=Multiply(
    A=StaticSwitchParameter(True=@n16, False=@n15) {parameterName=Icon, sortPriority=3}
    B=ScalarParameter() {defaultValue=1, parameterName=Opacity, sliderMax=1, sortPriority=1}
   )) {name=Opacity, nodeColor=<rgba 0 0 0 1>}
  NamedRerouteDeclaration(None=Multiply(A=@n8.RGB, B=@n8.RGB)) {name=Icon, nodeColor=<rgba 0 0 0 1>}
  n17 = TextureCoordinate()
  n18 = Multiply(A=ScalarParameter() {defaultValue=-0.03, group=Border, parameterName=Width - Right, sliderMax=1, sortPriority=35}) {constB=-1}
  n19 = ScalarParameter() {defaultValue=180, group=Border, parameterName=Angle - Right, sliderMax=360, sortPriority=35}
  n20 = Multiply(
   A=Rotator(Coordinate=Add(A=@n17, B=@n18), Time=Divide(A=@n19) {constB=360, desc=Conversion to decimal}) {speed=6.28}
   B=ScalarParameter() {defaultValue=200, group=Border, parameterName=Sharpness, sortPriority=30}
  )
  NamedRerouteDeclaration(None=LinearInterpolate(Alpha=Saturate(None=@n20) {desc=Clamp values between 0 and 1}) {constA=1, constAlpha=1, constB=0}) {name=Border Right, nodeColor=<rgba 0 0 0 1>}
  n21 = TextureCoordinate()
  n22 = Multiply(A=ScalarParameter() {defaultValue=0.03, group=Border, parameterName=Width - Top, sliderMax=1, sortPriority=35}) {constB=-1}
  n23 = ScalarParameter() {defaultValue=270, group=Border, parameterName=Angle - Top, sliderMax=360, sortPriority=35}
  n24 = Multiply(
   A=Rotator(Coordinate=Add(A=@n21, B=@n22), Time=Divide(A=@n23) {constB=360, desc=Conversion to decimal}) {speed=6.28}
   B=ScalarParameter() {defaultValue=200, group=Border, parameterName=Sharpness, sortPriority=30}
  )
  NamedRerouteDeclaration(None=LinearInterpolate(Alpha=Saturate(None=@n24) {desc=Clamp values between 0 and 1}) {constA=1, constB=0}) {name=Border Top, nodeColor=<rgba 0 0 0 1>}
  n25 = TextureCoordinate()
  n26 = Multiply(A=ScalarParameter() {defaultValue=-0.03, group=Border, parameterName=Width - Bottom, sliderMax=1, sortPriority=35}) {constB=-1}
  n27 = ScalarParameter() {defaultValue=90, group=Border, parameterName=Angle - Bottom, sliderMax=360, sortPriority=35}
  n28 = Multiply(
   A=Rotator(Coordinate=Add(A=@n25, B=@n26), Time=Divide(A=@n27) {constB=360, desc=Conversion to decimal}) {speed=6.28}
   B=ScalarParameter() {defaultValue=200, group=Border, parameterName=Sharpness, sortPriority=30}
  )
  NamedRerouteDeclaration(None=LinearInterpolate(Alpha=Saturate(None=@n28) {desc=Clamp values between 0 and 1}) {constA=1, constB=0}) {name=Border Bottom, nodeColor=<rgba 0 0 0 1>}
  n29 = Add(A=NamedRerouteUsage(), B=NamedRerouteUsage())
  n30 = Add(A=NamedRerouteUsage(), B=NamedRerouteUsage())
  NamedRerouteDeclaration(None=Clamp(None=Add(A=@n29, B=@n30))) {name=Border, nodeColor=<rgba 0 0 0 1>}
  n31 = VectorParameter() {parameterName=Color B, sortPriority=3, defaultValue=<rgba 1 0.2081 0.5587 1>}
  n32 = Multiply(A=@n31.RGB)
  n33 = Multiply(A=@n31.RGB) {constB=0}
  n34 = ScalarParameter() {defaultValue=0.5, group=Radial Gradient, parameterName=Position X, sliderMax=1, sortPriority=20}
  n35 = ScalarParameter() {defaultValue=0.5, group=Radial Gradient, parameterName=Position Y, sliderMax=1, sortPriority=20}
  n36 = NamedRerouteDeclaration(None=MaterialFunctionCall(
    UVs=TextureCoordinate()
    CenterPosition=AppendVector(A=@n34, B=@n35) {desc="x, y"}
    Radius=ScalarParameter() {defaultValue=0.5, group=Radial Gradient, parameterName=Radius, sliderMax=1, sortPriority=20}
    Density=ScalarParameter() {defaultValue=1, group=Radial Gradient, parameterName=Density, sortPriority=20}
   ) {materialFunction=&RadialGradientExponential}.RadialGradientExponential) {name=Radial Gradient Mask, nodeColor=<rgba 1 0.1878 0 1>}
  NamedRerouteDeclaration(None=LinearInterpolate(
    A=VectorParameter() {parameterName=Color A, sortPriority=3, defaultValue=<rgba 0.1159 0.1884 0.9271 1>}.RGB
    B=StaticSwitchParameter(True=@n32, False=@n33) {parameterName=Secondary Color, sortPriority=10}
    Alpha=OneMinus(None=@n36)
   )) {name=Radial Gradient, nodeColor=<rgba 0 0 0 1>}
  n37 = Add(
   A=TextureCoordinate()
   B=Multiply(A=ScalarParameter() {group=Linear Gradient, parameterName=Offset, sliderMax=1, sliderMin=-1, sortPriority=10}) {constB=-1}
  )
  n38 = Divide(A=ScalarParameter() {group=Linear Gradient, parameterName=Rotation, sliderMax=360, sortPriority=10}) {constB=360, desc=Conversion to decimal}
  n39 = Constant() {r=51}
  n40 = Divide(A=ScalarParameter() {defaultValue=100, group=Linear Gradient, parameterName=Lenght, sliderMax=100, sortPriority=10})
  n41 = Saturate(None=Multiply(A=Rotator(Coordinate=@n37, Time=@n38) {speed=6.3}, B=Subtract(A=@n39, B=@n40) {constB=50})) {desc=Clamp values between 0 and 1}
  NamedRerouteDeclaration(None=OneMinus(None=@n41)) {name=Linear Gradient Mask, nodeColor=<rgba 0 0 0 1>}
  n42 = VectorParameter() {parameterName=Color B, sortPriority=3, defaultValue=<rgba 1 0.2081 0.5587 1>}
  n43 = Multiply(A=@n42.RGB)
  n44 = Multiply(A=@n42.RGB) {constB=0}
  NamedRerouteDeclaration(None=LinearInterpolate(
    A=VectorParameter() {parameterName=Color A, sortPriority=3, defaultValue=<rgba 0.1159 0.1884 0.9271 1>}.RGB
    B=StaticSwitchParameter(True=@n43, False=@n44) {parameterName=Secondary Color, sortPriority=10}
    Alpha=@n41
   )) {name=Linear Gradient, nodeColor=<rgba 0 0 0 1>}
  NamedRerouteUsage()
  NamedRerouteUsage()
 ]
}
